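# Parent material for the Hacker backpack cosmetic, adding a blinking underscore cursor and scanlined screen effect to the standard character shading.
# It edits Base Color and Emissive between the CharacterParent 2023 PreFX and PostFX functions via named reroutes.
# BoxMask-2D calls define screen and underscore regions in UV space, with a rounded time sine toggling the tinted underscore.
# A shared scanline function uses vertex-interpolated local XZ position as UVs.
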Material /BRCosmetics/Accessories/FORT_Backpacks/Backpack_F_MED_Hacker/Materials/M_Hacker_Backpack_Parent
graph {
 properties {
  bUseMaterialAttributes=true
  bUsedWithMeshDeformer=true
  bUsedWithMorphTargets=true
  bUsedWithSkeletalMesh=true
  pixelDepthOffsetMode=PDOM_Legacy
  shadingModel=MSM_FromMaterialExpression
  usedShadingModels=Default Lit | Subsurface | Clear Coat | Subsurface Profile | Cloth
 }
 nodes [
  n1 = MaterialFunctionCall!(A=TextureCoordinate(), B=Constant2Vector() {g=0.056, r=0.094}, Bounds=Constant2Vector() {g=0.065, r=0.1875}, Edge Falloff=Constant()) {materialFunction=/Engine/Functions/Engine_MaterialFunctions01/Gradient/BoxMask-2D}
  LinearInterpolate(A=TextureSample() {texture=/BRCosmetics/Accessories/FORT_Backpacks/Backpack_F_MED_Hacker/Textures/T_Hacker_Backpack_E}.RGB, B=@n1.Result)
  n2 = MaterialFunctionCall() {materialFunction=&MF_CharacterParent_2023_PreFX}
  MaterialFunctionCall(Material=NamedRerouteUsage(), M=@n2.M, Pre Skinned Local Position=@n2.M, Distance Blend=@n2.M, Fresnel=@n2.M) {materialFunction=&MF_CharacterParent_2023_PostFX}
  NamedRerouteDeclaration(None=@n2.Result) {name=InMaterial, nodeColor=<rgba 0 0.3672 1 1>}
  n3 = GetMaterialAttributes(None=NamedRerouteUsage()) {attributeGetTypes [#1, #2]}
  NamedRerouteDeclaration(None=@n3.Base Color) {name=InBaseColor, nodeColor=<rgba 0.6875 1 0 1>}
  NamedRerouteDeclaration(None=@n3.Emissive Color) {name=InEmissiveColor, nodeColor=<rgba 0.9922 0 1 1>}
  n4 = Constant() {r=1}
  n5 = ScalarParameter() {defaultValue=7, parameterName=ScanlineSize}
  n6 = Constant() {r=0.3}
  n7 = LocalPosition()
  n8 = VertexInterpolator(VS=Divide(A=ComponentMask(None=@n7.XYZ) {b=true, r=true}) {constB=100})
  n9 = VectorParameter() {parameterName=UnderscoreColor, defaultValue=<rgba 0.0627 0.0918 0.7084 1>}
  n10 = LinearInterpolate(A=NamedRerouteUsage(), B=@n9.RGB, Alpha=NamedRerouteUsage())
  NamedRerouteDeclaration(None=LinearInterpolate(
    A=NamedRerouteUsage()
    B=MaterialFunctionCall!(ScanlineIntensity=@n4, ScanlineThickness, ScanlineScale=@n5, ScanNegativeIntensity=@n6, UVs=@n8.PS, In=@n10) {materialFunction=/Game/Characters/Player/TechArt/Superlevels/S30/MaterialFunctions/MF_SL_S30_Scanlines}.Result
    Alpha=NamedRerouteUsage()
   )) {name=OutEmissive, nodeColor=<rgba 0 1 0.6719 1>}
  n11 = TextureCoordinate()
  n12 = Constant2Vector() {g=0.0625, r=0.152}
  n13 = Constant2Vector() {g=0.0075, r=0.027}
  n14 = Constant()
  n15 = Time()
  n16 = ConstantBiasScale(None=Sine(None=Multiply(A=@n15) {constB=0.9}))
  NamedRerouteDeclaration(None=Multiply(
    A=MaterialFunctionCall(A=@n11, B=@n12, Bounds=@n13, Edge Falloff=@n14) {materialFunction=&BoxMask-2D}.Result
    B=Round(None=@n16)
   )) {name=UnderscoreMask, nodeColor=<rgba 1 0.3516 0 1>}
  NamedRerouteDeclaration(None=LinearInterpolate(A=NamedRerouteUsage(), B=Multiply(A=@n9.RGB) {constB=0.5}, Alpha=NamedRerouteUsage())) {name=OutBaseColor, nodeColor=<rgba 0.0312 0 1 1>}
  NamedRerouteDeclaration(None=@n1.Result) {name=ScreenMask, nodeColor=<rgba 0 0.7891 1 1>}
  NamedRerouteDeclaration(None=SetMaterialAttributes!(MaterialAttributes=@n3.MaterialAttributes, Base Color=NamedRerouteUsage(), Emissive Color=NamedRerouteUsage()) {attributeSetTypes [#1, #2]}) {name=OutMaterial, nodeColor=<rgba 0 0.5312 1 1>}
 ]
}
